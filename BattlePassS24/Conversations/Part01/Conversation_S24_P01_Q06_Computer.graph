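# Conversation database for a Season 24 Part 1 Quest 6 computer, styled as a file browser with corrupted OPPO files, a cookbook draft and personnel profiles.
# The main entry auto-selects into the directory list, where speech nodes with choice sub-nodes act as folders, files and pages.
# RETURN choices use link nodes to re-enter menus by entry tag.
# CLOSE ALL FILES, offered inside opened files, ends the conversation and fires the quest prop success side effect.
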ConversationDatabase /BattlePassS24/Conversations/Part01/Conversation_S24_P01_Q06_Computer
entry_points [
 {tag=Conversation.Entry.Season24.Narrative.P01.Q06, destinations [#1]}
 {tag=Conversation.Entry.Season24.Narrative.P01.Q06.Return, destinations [#2]}
 {tag=Conversation.Entry.Season24.Narrative.P01.Q06.Personnel, destinations [#3]}
 {tag=Conversation.Entry.Season24.Narrative.P01.Q06.Oppo, destinations [#4]}
]
nodes [
 ConversationEntryPointNode_3 {
  id=#4
  class=ConversationEntryPointNode
  outputs [#5, #6, #7]
  properties {entryTag {tagName=Conversation.Entry.Season24.Narrative.P01.Q06.Oppo}}
 }
 Speech_C_23 {
  id=#7
  class=Speech_C
  outputs [#8]
  properties {softTaskIcon=/Game/UI/Foundation/Textures/Icons/Replay/T-Icon-Replay-Rewind-L, generalConfig {defaultMessage=SELECT DIRECTORY}}
  sub_nodes [ConversationChoiceNode_14 {class=ConversationChoiceNode, properties {defaultChoiceDisplayText=RETURN}}]
 }
 ConversationLinkNode_5 {
  id=#8
  class=ConversationLinkNode
  properties {remoteEntryTag {tagName=Conversation.Entry.Season24.Narrative.P01.Q06.Return}}
 }
 Speech_C_17 {
  id=#6
  class=Speech_C
  outputs [#9, #10]
  properties {softTaskIcon=/Game/UI/Foundation/Textures/Icons/Replay/T-Icon-Replay-Folder-L, generalConfig {defaultMessage="[[FILE CORRUPTED]]"}}
  sub_nodes [ConversationChoiceNode_2 {class=ConversationChoiceNode, properties {defaultChoiceDisplayText="[[FILE CORRUPTED]]"}}]
 }
 Speech_C_12 {
  id=#10
  class=Speech_C
  outputs []
  properties {softTaskIcon=/Game/UI/Foundation/Textures/Icons/Replay/T-Icon-Replay-Filter-Death-M, generalConfig {defaultMessage=CLOSING FILES...}}
  sub_nodes [
   ConversationChoiceNode_20 {class=ConversationChoiceNode, properties {defaultChoiceDisplayText=CLOSE ALL FILES}}
   ConvoSideEffect_QuestPropActor_ConvoSuccess_C_1 {class=ConvoSideEffect_QuestPropActor_ConvoSuccess_C}
  ]
 }
 Speech_C_21 {
  id=#9
  class=Speech_C
  outputs [#11]
  properties {softTaskIcon=&T-Icon-Replay-Rewind-L, generalConfig {defaultMessage=SELECT DIRECTORY}}
  sub_nodes [ConversationChoiceNode_18 {class=ConversationChoiceNode, properties {defaultChoiceDisplayText=RETURN}}]
 }
 ConversationLinkNode_2 {
  id=#11
  class=ConversationLinkNode
  properties {remoteEntryTag {tagName=Conversation.Entry.Season24.Narrative.P01.Q06.Oppo}}
 }
 Speech_C_18 {
  id=#5
  class=Speech_C
  outputs [#9, #10]
  properties {softTaskIcon=&T-Icon-Replay-Folder-L, generalConfig {defaultMessage="[[FILE CORRUPTED]]"}}
  sub_nodes [ConversationChoiceNode_19 {class=ConversationChoiceNode, properties {defaultChoiceDisplayText="[[FILE CORRUPTED]]"}}]
 }
 ConversationEntryPointNode_1 {
  id=#3
  class=ConversationEntryPointNode
  outputs [
   #12
   #13
   #14
   #15
   #16
  ]
  properties {entryTag {tagName=Conversation.Entry.Season24.Narrative.P01.Q06.Personnel}}
 }
 Speech_C_22 {
  id=#16
  class=Speech_C
  outputs [#17]
  properties {softTaskIcon=&T-Icon-Replay-Rewind-L, generalConfig {defaultMessage=SELECT DIRECTORY}}
  sub_nodes [ConversationChoiceNode_24 {class=ConversationChoiceNode, properties {defaultChoiceDisplayText=RETURN}}]
 }
 ConversationLinkNode_3 {
  id=#17
  class=ConversationLinkNode
  properties {remoteEntryTag {tagName=Conversation.Entry.Season24.Narrative.P01.Q06.Return}}
 }
 Speech_C_3 {
  id=#15
  class=Speech_C
  outputs [#18, #19]
  properties {
   softTaskIcon=&T-Icon-Replay-Folder-L
   generalConfig {defaultMessage=PEELY--- Profile Type: Potential Recruit. Status: Pending approach.}
  }
  sub_nodes [ConversationChoiceNode_11 {class=ConversationChoiceNode, properties {defaultChoiceDisplayText=PEELY}}]
 }
 Speech_C_20 {
  id=#19
  class=Speech_C
  outputs []
  properties {softTaskIcon=&T-Icon-Replay-Filter-Death-M, generalConfig {defaultMessage=CLOSING FILES...}}
  sub_nodes [
   ConversationChoiceNode_15 {class=ConversationChoiceNode, properties {defaultChoiceDisplayText=CLOSE ALL FILES}}
   ConvoSideEffect_QuestPropActor_ConvoSuccess_C_2 {class=ConvoSideEffect_QuestPropActor_ConvoSuccess_C}
  ]
 }
 Speech_C_13 {
  id=#18
  class=Speech_C
  outputs [#20]
  properties {softTaskIcon=&T-Icon-Replay-Rewind-L, generalConfig {defaultMessage=PERSONNEL}}
  sub_nodes [ConversationChoiceNode_23 {class=ConversationChoiceNode, properties {defaultChoiceDisplayText=RETURN}}]
 }
 ConversationLinkNode_0 {
  id=#20
  class=ConversationLinkNode
  properties {remoteEntryTag {tagName=Conversation.Entry.Season24.Narrative.P01.Q06.Personnel}}
 }
 Speech_C_2 {
  id=#14
  class=Speech_C
  outputs [#18, #19]
  properties {
   softTaskIcon=&T-Icon-Replay-Folder-L
   generalConfig {defaultMessage=STASH'D --- Profile Type: Contractor. Status: Pending promotion.}
  }
  sub_nodes [ConversationChoiceNode_3 {class=ConversationChoiceNode, properties {defaultChoiceDisplayText=STASH'D}}]
 }
 Speech_C_1 {
  id=#13
  class=Speech_C
  outputs [#18, #19]
  properties {
   softTaskIcon=&T-Icon-Replay-Folder-L
   generalConfig {defaultMessage="TWYN--- Profile Type: Field Agent. Status: On assignment. "}
  }
  sub_nodes [ConversationChoiceNode_16 {class=ConversationChoiceNode, properties {defaultChoiceDisplayText=TWYN}}]
 }
 Speech_C_10 {
  id=#12
  class=Speech_C
  outputs [#18, #19]
  properties {
   softTaskIcon=&T-Icon-Replay-Folder-L
   generalConfig {defaultMessage="EVIE--- Profile Type: Full Clearance. Status: On probation. "}
  }
  sub_nodes [ConversationChoiceNode_6 {class=ConversationChoiceNode, properties {defaultChoiceDisplayText=EVIE}}]
 }
 ConversationEntryPointNode_0 {
  id=#2
  class=ConversationEntryPointNode
  outputs [#21, #22, #23]
  properties {entryTag {tagName=Conversation.Entry.Season24.Narrative.P01.Q06.Return}}
 }
 Speech_C_0 {
  id=#23
  class=Speech_C
  outputs [
   #12
   #13
   #14
   #15
   #16
  ]
  properties {softTaskIcon=&T-Icon-Replay-Folder-L, generalConfig {defaultMessage=PERSONNEL}}
  sub_nodes [ConversationChoiceNode_10 {class=ConversationChoiceNode, properties {defaultChoiceDisplayText=PERSONNEL}}]
 }
 Speech_C_7 {
  id=#22
  class=Speech_C
  outputs [#24]
  properties {
   softTaskIcon=&T-Icon-Replay-Folder-L
   generalConfig {defaultMessage="...Shredded slurpshroom. Stir until rubbery, then add fireflies..."}
  }
  sub_nodes [ConversationChoiceNode_5 {class=ConversationChoiceNode, properties {defaultChoiceDisplayText=COOKBOOK DRAFT}}]
 }
 Speech_C_9 {
  id=#24
  class=Speech_C
  outputs [#25]
  properties {generalConfig {defaultMessage=...the Slap-based glaze brings out the electricity in the dish...}}
  sub_nodes [ConversationChoiceNode_7 {class=ConversationChoiceNode, properties {defaultChoiceDisplayText=Continue}}]
 }
 Speech_C_11 {
  id=#25
  class=Speech_C
  outputs [#26]
  properties {generalConfig {defaultMessage="..When I was a young lizard, a cabbage was my best friend..."}}
  sub_nodes [ConversationChoiceNode_8 {class=ConversationChoiceNode, properties {defaultChoiceDisplayText=Continue}}]
 }
 Speech_C_14 {
  id=#26
  class=Speech_C
  outputs [#27, #28]
  properties {generalConfig {defaultMessage=...And add corn. Seriously. As much corn as you can find. All of it...}}
  sub_nodes [ConversationChoiceNode_1 {class=ConversationChoiceNode, properties {defaultChoiceDisplayText=Continue}}]
 }
 Speech_C_19 {
  id=#28
  class=Speech_C
  outputs []
  properties {softTaskIcon=&T-Icon-Replay-Filter-Death-M, generalConfig {defaultMessage=CLOSING FILES...}}
  sub_nodes [
   ConversationChoiceNode_21 {class=ConversationChoiceNode, properties {defaultChoiceDisplayText=CLOSE ALL FILES}}
   ConvoSideEffect_QuestPropActor_ConvoSuccess_C_3 {class=ConvoSideEffect_QuestPropActor_ConvoSuccess_C}
  ]
 }
 Speech_C_15 {
  id=#27
  class=Speech_C
  outputs [#29]
  properties {softTaskIcon=&T-Icon-Replay-Rewind-L, generalConfig {defaultMessage=SELECT DIRECTORY}}
  sub_nodes [ConversationChoiceNode_22 {class=ConversationChoiceNode, properties {defaultChoiceDisplayText=RETURN}}]
 }
 ConversationLinkNode_1 {
  id=#29
  class=ConversationLinkNode
  properties {remoteEntryTag {tagName=Conversation.Entry.Season24.Narrative.P01.Q06.Return}}
 }
 Speech_C_6 {
  id=#21
  class=Speech_C
  outputs [#5, #6, #7]
  properties {softTaskIcon=&T-Icon-Replay-Folder-L, generalConfig {defaultMessage="[[FILE CORRUPTED]]"}}
  sub_nodes [ConversationChoiceNode_13 {class=ConversationChoiceNode, properties {defaultChoiceDisplayText=OPPO}}]
 }
 ConversationEntryPointNode_2 {
  id=#1
  class=ConversationEntryPointNode
  outputs [#30]
  properties {entryTag {tagName=Conversation.Entry.Season24.Narrative.P01.Q06}}
 }
 Speech_C_5 {
  id=#30
  class=Speech_C
  outputs [#21, #22, #23]
  properties {generalConfig {defaultMessage=SELECT DIRECTORY}}
  sub_nodes [ConversationChoiceNode_4 {
    class=ConversationChoiceNode
    properties {defaultChoiceDisplayText=SELECT DIRECTORY, choiceTags {gameplayTags [{tagName=Conversation.Choice.AutoSelectIfValid}]}}
   }]
 }
]
