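# Conversation database for Season 24 narrative quest Part 01 Q07, in which Evie and Thunder urge uniting the Syndicates to expose a saboteur.
# Each NPC has its own entry point leading to an auto-selecting main-questline choice and two chained speech lines.
# The first line requires an active, uncompleted quest, that NPC's objective still open and a matching NPC conversation tag.
# Each branch ends by completing its objective: obj0 for Evie, obj1 for Thunder.
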ConversationDatabase /BattlePassS24/Conversations/Part01/Conversation_S24_P01_Q07
entry_points [
 {tag=Conversation.Entry.Season24.Narrative.P01.Q07.Evie, destinations [#1]}
 {tag=Conversation.Entry.Season24.Narrative.P01.Q07.Thunder, destinations [#2]}
]
nodes [
 ConversationEntryPointNode_3 {
  id=#2
  class=ConversationEntryPointNode
  outputs [#3]
  properties {entryTag {tagName=Conversation.Entry.Season24.Narrative.P01.Q07.Thunder}}
 }
 Speech_C_0 {
  id=#3
  class=Speech_C
  outputs [#4]
  properties {
   softTaskIcon=&T_UI_NPC_MainQuest
   generalConfig {defaultMessage="We need to win the trust of each Syndicate, one by one."}
  }
  sub_nodes [
   ConversationChoiceNode_3 {
    class=ConversationChoiceNode
    properties {
     defaultChoiceDisplayText=Talk to Thunder
     choiceTags {gameplayTags [{tagName=Conversation.Choice.AutoSelectIfValid}, {tagName=Conversation.Choice.Category.MainQuestline}]}
    }
   }
   FortConversationRequirement_ControllerRequirement_12 {
    class=FortConversationRequirement_ControllerRequirement
    properties {
     nodeName=Require: Has Completed Quest
     requirement {
      class=FortControllerRequirement_HasCompletedQuest
      properties {bInverseCondition=true, questDefinition=/BattlePassS24/Items/QuestItems/NarrativeQuests/Part01/quest_s24_narrative_p01_q07}
     }
    }
   }
   FortConversationRequirement_ControllerRequirement_13 {
    class=FortConversationRequirement_ControllerRequirement
    properties {
     nodeName=Require: Is Quest Active
     requirement {
      class=FortControllerRequirement_IsQuestActive
      properties {questDefinition=&quest_s24_narrative_p01_q07}
     }
    }
   }
   FortConversationRequirement_ControllerRequirement_14 {
    class=FortConversationRequirement_ControllerRequirement
    properties {
     nodeName=Require: Has Completed Quest Objective
     requirement {
      class=FortControllerRequirement_HasCompletedQuestObjective
      properties {
       bInverseCondition=true
       objectiveBackendName=quest_s24_narrative_p01_q07_obj1
       questDefinition=&quest_s24_narrative_p01_q07
      }
     }
    }
   }
   FortConversationRequirement_ControllerRequirement_16 {
    class=FortConversationRequirement_ControllerRequirement
    properties {
     nodeName=Require: NPC Has Conversation Tag
     participantId {tagName=Conversation.Participant.NPC}
     requirement {class=FortControllerRequirement_NPCHasConversationEntryTag, properties {queryTag {tagName=Conversation.Entry.Thunder}}}
    }
   }
  ]
 }
 Speech_C_1 {
  id=#4
  class=Speech_C
  outputs [#5]
  properties {generalConfig {defaultMessage="When we're all united, the saboteur will have nowhere to hide."}}
  sub_nodes [ConversationChoiceNode_7 {class=ConversationChoiceNode, properties {defaultChoiceDisplayText=Continue}}]
 }
 FortConversationTaskNode_CompleteQuestObjectiveByBackendName_0 {
  id=#5
  class=FortConversationTaskNode_CompleteQuestObjectiveByBackendName
  properties {
   objectiveBackendName=quest_s24_narrative_p01_q07_obj1
   questDefinition=&quest_s24_narrative_p01_q07
  }
 }
 ConversationEntryPointNode_0 {
  id=#1
  class=ConversationEntryPointNode
  outputs [#6]
  properties {entryTag {tagName=Conversation.Entry.Season24.Narrative.P01.Q07.Evie}}
 }
 Speech_C_3 {
  id=#6
  class=Speech_C
  outputs [#7]
  properties {
   softTaskIcon=&T_UI_NPC_MainQuest
   generalConfig {defaultMessage="We have to win the trust of each Syndicate, one by one."}
  }
  sub_nodes [
   ConversationChoiceNode_1 {
    class=ConversationChoiceNode
    properties {
     defaultChoiceDisplayText=Talk to Evie
     choiceTags {gameplayTags [{tagName=Conversation.Choice.AutoSelectIfValid}, {tagName=Conversation.Choice.Category.MainQuestline}]}
    }
   }
   FortConversationRequirement_ControllerRequirement_0 {
    class=FortConversationRequirement_ControllerRequirement
    properties {
     nodeName=Require: Has Completed Quest
     requirement {
      class=FortControllerRequirement_HasCompletedQuest
      properties {bInverseCondition=true, questDefinition=&quest_s24_narrative_p01_q07}
     }
    }
   }
   FortConversationRequirement_ControllerRequirement_1 {
    class=FortConversationRequirement_ControllerRequirement
    properties {
     nodeName=Require: Is Quest Active
     requirement {
      class=FortControllerRequirement_IsQuestActive
      properties {questDefinition=&quest_s24_narrative_p01_q07}
     }
    }
   }
   FortConversationRequirement_ControllerRequirement_2 {
    class=FortConversationRequirement_ControllerRequirement
    properties {
     nodeName=Require: Has Completed Quest Objective
     requirement {
      class=FortControllerRequirement_HasCompletedQuestObjective
      properties {
       bInverseCondition=true
       objectiveBackendName=quest_s24_narrative_p01_q07_obj0
       questDefinition=&quest_s24_narrative_p01_q07
      }
     }
    }
   }
   FortConversationRequirement_ControllerRequirement_3 {
    class=FortConversationRequirement_ControllerRequirement
    properties {
     nodeName=Require: NPC Has Conversation Tag
     participantId {tagName=Conversation.Participant.NPC}
     requirement {class=FortControllerRequirement_NPCHasConversationEntryTag, properties {queryTag {tagName=Conversation.Entry.Evie}}}
    }
   }
  ]
 }
 Speech_C_28 {
  id=#7
  class=Speech_C
  outputs [#8]
  properties {generalConfig {defaultMessage="We gotta unite as a city, and find this saboteur before it's too late."}}
  sub_nodes [ConversationChoiceNode_4 {class=ConversationChoiceNode, properties {defaultChoiceDisplayText=Continue}}]
 }
 FortConversationTaskNode_CompleteQuestObjectiveByBackendName_1 {
  id=#8
  class=FortConversationTaskNode_CompleteQuestObjectiveByBackendName
  properties {
   objectiveBackendName=quest_s24_narrative_p01_q07_obj0
   questDefinition=&quest_s24_narrative_p01_q07
  }
 }
]
